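# Sound cue for a generic building pop in the HomeBase UI, routed to the Fort_UI sound class.
# A single wave player feeds the root directly with the homebase_mid_pop wave, with a boosted volume multiplier and no randomization or attenuation nodes.
# It overrides concurrency with a PreventNew rule, so new pops are refused once the instance limit is reached rather than cutting off those already playing.
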SoundCue /Game/Sounds/Fort_UI_Sounds/HomeBase/HomeBase_Generic_Building_Pop_Cue
root=WavePlayer_1
nodes [{id=WavePlayer_1, properties {soundWaveAssetPtr=&/homebase_mid_pop}}]
properties {
 bOverrideConcurrency=true
 maxDistance=2097152
 soundClassObject=&Fort_UI
 volumeMultiplier=1.4
 concurrencyOverrides {resolutionRule=PreventNew}
}
